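; Parent Blueprint for construction pallet stack props, building a configurable-height stack from one instanced static mesh.
; The construction script spaces PalletCount instances by the mesh's bounds height, applies PalletMaterial, and gives each a random yaw within ±RotationAngle from a location-seeded stream.
; Asset-provider overrides allow the material to be referenced in UEFN.
; Collision serves as the Curie collision override, and destruction removes the box component.
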(event EventBeginPlay)

(event Collision|EventActorBeginOverlap
  (params (OtherActor :type Actor%20Object%20Reference)))

(event EventTick
  (params (DeltaSeconds :type Float%20%28single-precision%29)))

(event EventDestroyed
  (Components|DestroyComponent (Actor|GetComponentbyClass "/Script/Engine.BoxComponent")))

(fn BuildExposedAssetProviderVersion(Implementation)
  (params (Version :type Fort%20Exposed%20Asset%20Provider%20Version%20Structure%20%28by%20ref%29))
  (|CallParentFunction Version)
  (ExposedAssetProvider|AddVersion Version 1)
  (return))

(fn CollectAssetReferencestoAllow(Implementation)
  (params (Collector :type Fort%20Exposed%20Asset%20Reference%20Collector%20Structure%20%28by%20ref%29))
  (|CallParentFunction Collector)
  (ExposedAssetProvider|CollectObjectPropertyAssetReferences Collector 0 (Variables|Default|GetPalletMaterial)) ; Collect the material to allow it to be referenced in UEFN
  (return))

(fn GetCurieCollisionComponentOverride
  (return (Variables|Default|GetCollision)))

(fn ConstructionScript
  (bind (_origin _boxextent) (Math|BoxSphereBounds|BreakBoxSphereBounds (StaticMesh|GetBounds (Variables|/Script/Engine.StaticMeshComponent|GetStaticMesh (Variables|Default|GetInstancedStaticMesh)))))
  (Variables|Default|SetRandomStream (Math|Random|MakeRandomStreamfromLocation (Transformation|GetActorLocation) 1.0 true))
  (Variables|Default|SetHeightOffset (* (.z _boxextent) 2.0))
  (for _index (range (Variables|Default|GetPalletCount))
    (Components|InstancedStaticMesh|AddInstance (Variables|Default|GetInstancedStaticMesh) (Math|Transform|MakeTransform (Math|Conversions|ToVector(IntVector) (Utilities|Struct|MakeIntVector 0 0 (Math|Float|Truncate (* _index (Variables|Default|GetHeightOffset))))) (Math|Rotator|MakeRotator 0.0 0.0 (Math|Random|RandomFloatinRangefromStream (Variables|Default|GetRandomStream) (- (Variables|Default|GetRotationAngle) (* (Variables|Default|GetRotationAngle) 2.0)) (Variables|Default|GetRotationAngle)))) false)
    (Rendering|Material|SetMaterial (Variables|Default|GetInstancedStaticMesh) 0 (Variables|Default|GetPalletMaterial))))
